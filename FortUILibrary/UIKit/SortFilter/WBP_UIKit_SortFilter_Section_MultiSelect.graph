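# UIKit sort/filter section widget for multi-select options, showing a title above scrollable toggle entries in an evenly spaced grid or wrap box.
# SetEntries reuses pooled entries when their count matches the visible item view models, otherwise rebuilds them.
# Zero-count entries are disabled and dimmed, the first usable one becomes focus target, and clicks re-broadcast through OnEntryClicked.
# PreConstruct applies title, columns and mobile padding, plus a design-time preview.
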WidgetBlueprint /FortUILibrary/UIKit/SortFilter/WBP_UIKit_SortFilter_Section_MultiSelect
graphs [
 CreateEntries {nodes [
   CreateEntries!() -> (then, SortFilterItemVMs:Array of Fort Sort Filter Item VM Object References) {is=title K2Node_FunctionEntry}
   > ResetEntries()
   > SetActiveWidget(
    self=Get Switcher_Entries() -> (Switcher_Entries:Common Visibility Switcher Object Reference)
    Widget=Select(
     Option 0:Widget Object Reference=Get DynamicBox_Entries() -> (DynamicBox_Entries:Dynamic Entry Box Object Reference)
     Option 1:Widget Object Reference=Get Grid_Entries() -> (Grid_Entries:UIKit Grid Panel Object Reference)
     Index=Get UseEvenlySpacedColumns() -> (UseEvenlySpacedColumns:bool)
    ) -> (ReturnValue:Widget Object Reference)
   )
   > n1 = For Each Loop(Array:Array of Fort Sort Filter Item VM Object References=Get SortFilterItemVMs() -> (SortFilterItemVMs:Array of Fort Sort Filter Item VM Object References)) -> (Array Element:Fort Sort Filter Item VM Object Reference)
   > Branch(Condition=Get bIsVisible!(self:Fort Sort Filter Item VM Object Reference=@n1) -> (bIsVisible:bool))
   > n2 = Sequence!(execute) -> (then_0, then_1, then_2:Exec)
   > n3 = Branch(Condition=Get UseEvenlySpacedColumns())
   > n4 = CreateChildOnGrid!(
    execute
    self:UIKit Grid Panel Object Reference=Get Grid_Entries()
    WidgetClass:"User Widget Class Reference (by ref)"=Get EntryWidgetClass() -> (EntryWidgetClass:User Widget Class Reference)
    InRow:int=Get RowIndex() -> (RowIndex:int=0)
    InColumn:int=Get ColumnIndex() -> (ColumnIndex:int=0)
   ) -> (then, ReturnValue:Grid Slot Object Reference)
   > n5 = Cast To WBP_UIKit_SortFilter_Entry!(execute, Object=GetContent!(self:Panel Slot Object Reference=@n4) -> (ReturnValue:Widget Object Reference)) -> (then, CastFailed, AsWBP UIKit Sort Filter Entry:WBP UIKit Sort Filter Entry Object Reference) {is=title K2Node_DynamicCast}
   > Set LoopWidget(LoopWidget:WBP UIKit Sort Filter Entry Object Reference=@n5) -> (Output_Get:WBP UIKit Sort Filter Entry Object Reference)
   n6 = Create Entry!(execute=@n3.else, self:Dynamic Entry Box Object Reference=Get DynamicBox_Entries()) -> (then, ReturnValue:User Widget Object Reference)
   > n7 = Cast To WBP_UIKit_SortFilter_Entry(Object=@n6)
   > Set LoopWidget(LoopWidget=@n7)
   Add(
    execute=@n2.then_1
    TargetArray:Array of WBP UIKit Sort Filter Entry Object References=Get EntryPool() -> (EntryPool:Array of WBP UIKit Sort Filter Entry Object References)
    NewItem:"WBP UIKit Sort Filter Entry Object Reference (by ref)"=Get LoopWidget() -> (LoopWidget:WBP UIKit Sort Filter Entry Object Reference)
   ) {is=title K2Node_CallArrayFunction}
   > n8 = SetEntryState!(
    execute
    self
    InputVM:Fort Sort Filter Item VM Object Reference=@n1
    InputEntry:WBP UIKit Sort Filter Entry Object Reference=Get LoopWidget()
   ) -> (then, UsableEntry:WBP UIKit Sort Filter Entry Object Reference)
   > n9 = Get!(execute:Exec) -> (then:Exec, else:Exec, FirstEntry:Widget Object Reference)
   Is Valid(exec=@n9.else, InputObject=@n8)
   > Set FirstEntry(FirstEntry:Widget Object Reference=@n8) -> (Output_Get:Widget Object Reference)
   n10 = Increment Int(" "=@n2.then_2, Value=Get ColumnIndex() -> (ColumnIndex=0))
   > Branch(Condition="integer >= integer"(A=@n10, B=Get ColumnCount() -> (ColumnCount:int=0)))
   > Increment Int(Value=Get RowIndex() -> (RowIndex=0))
   > Set ColumnIndex(ColumnIndex:int=0) -> (Output_Get:int=0)
  ]}
 DesignTimePreview {nodes [
   DesignTimePreview() {is=title K2Node_FunctionEntry}
   > ResetEntries()
   > n1 = For Loop(FirstIndex=1, LastIndex=Get DesignPreviewCount() -> (DesignPreviewCount:int=0))
   > n2 = Sequence()
   SetToggleType!(
    execute=@n2.then_1
    self:WBP UIKit Button With Toggle Object Reference=Get WBP_UIKit_ButtonWithToggle!(self:WBP UIKit Sort Filter Entry Object Reference=Get LoopWidget()) -> (WBP_UIKit_ButtonWithToggle:WBP UIKit Button With Toggle Object Reference)
    Type:E_UI_ToggleType Enum=NewEnumerator1
   ) -> (then)
   n3 = Select(
    Option 0:Text="LOCTEXT(\"Filter option\")"
    Option 1:Text="LOCTEXT(\"Filter option title that\\'s very long and goes to multiple lines\")"
    Index="Equal (Integer)"(A=@n1, B=1)
   ) -> (ReturnValue:Text)
   > SetText!(
    execute
    self:WBP UIKit Button Generic Base Object Reference=Get WBP_UIKit_ButtonWithToggle(self=Get LoopWidget())
    Text:Text=Select(
     Option 0=@n3
     Option 1=Format Text!(Format="LOCTEXT(\"{Label} (0)\")", Label:Text=@n3) -> (Result)
     Index=Get ShowFilterCounts() -> (ShowFilterCounts:bool)
    )
   ) -> (then)
   n4 = Increment Int(" "=@n2.then_2, Value=Get ColumnIndex() -> (ColumnIndex=0))
   > Branch(Condition="integer >= integer"(A=@n4, B=Get ColumnCount() -> (ColumnCount=0)))
   > Increment Int(Value=Get RowIndex() -> (RowIndex=0))
   > Set ColumnIndex(ColumnIndex=0) -> (Output_Get=0)
   n5 = CreateChildOnGrid(
    execute=@n2.then_0
    self=Get Grid_Entries()
    WidgetClass=Get EntryWidgetClass()
    InRow=Get RowIndex() -> (RowIndex=0)
    InColumn=Get ColumnIndex() -> (ColumnIndex=0)
   )
   > n6 = Cast To WBP_UIKit_SortFilter_Entry(Object=GetContent(self=@n5))
   > Set LoopWidget(LoopWidget=@n6)
  ]}
 EventGraph {nodes [
   n1 = Handle_OnClicked!() -> (OutputDelegate, then, SortFilterItemVM:Fort Sort Filter Item VM Object Reference) {is=title K2Node_CustomEvent}
   > Call On Entry Clicked!(
    execute:Exec
    self:WBP UIKit Sort Filter Section Multi Select Object Reference
    SortFilterItemVM:Fort Sort Filter Item VM Object Reference=@n1.SortFilterItemVM
   ) -> (then:Exec) {is=title K2Node_CallDelegate}
   n2 = Event PreConstruct()
   > Set Text(Text=Get SectionTitle() -> (SectionTitle:Text), self=Get Text_Title() -> (Text_Title:Common Text Object Reference))
   > SetColumnCount!(execute, self, Columns:int=Get ColumnCount() -> (ColumnCount=0)) -> (then)
   > SetMobileLayout()
   > Branch(Condition=@n2.IsDesignTime)
   > DesignTimePreview() {is=title K2Node_CallFunction}
  ]}
 GetDesiredFocusTarget {nodes [
   GetDesiredFocusTarget() {is=title K2Node_FunctionEntry}
   > Return Node!(execute, FocusTarget:Widget Object Reference=Get FirstEntry() -> (FirstEntry:Widget Object Reference))
  ]}
 HasRequiredEntryCount {nodes [
   n1 = HasRequiredEntryCount!() -> (then, FilterVMs:Array of Fort Sort Filter Item VM Object References) {is=title K2Node_FunctionEntry}
   > n2 = For Each Loop(Array=@n1)
   n3 = Get EntryPool()
   n4 = Get NeededEntryCount() -> (NeededEntryCount:int=0)
   Return Node!(
    execute=@n2.Completed
    HasRequiredEntries:bool="Equal (Integer)"(A=@n4, B=Length(TargetArray:Array of WBP UIKit Sort Filter Entry Object References=@n3))
   )
   > Branch(Condition=Get bIsVisible(self=@n2))
   > Increment Int(Value=@n4)
  ]}
 OnEntryClicked {nodes [OnEntryClicked!() -> (then, SortFilterItemVM:Fort Sort Filter Item VM Object Reference) {is=title K2Node_FunctionEntry}]}
 ResetEntries {nodes [
   ResetEntries() {is=title K2Node_FunctionEntry}
   > n1 = Set EntriesToClear(EntriesToClear:Array of Widget Object References=GetAllChildren(self=Get Grid_Entries())) -> (Output_Get:Array of Widget Object References)
   > Append Array!(
    execute
    TargetArray:Array of Widget Object References=@n1
    SourceArray:Array of Widget Object References=GetAllEntries!(self:Dynamic Entry Box Base Object Reference=Get DynamicBox_Entries()) -> (ReturnValue:Array of User Widget Object References)
   ) -> (then)
   > n2 = For Each Loop(Array:Array of Widget Object References=Get EntriesToClear() -> (EntriesToClear:Array of Widget Object References)) -> (Array Element:Widget Object Reference)
   ClearChildren(execute=@n2.Completed, self=Get Grid_Entries())
   > Reset!(execute, self:Dynamic Entry Box Object Reference=Get DynamicBox_Entries(), bDeleteWidgets:bool=false) -> (then)
   > Set FirstEntry()
   > Clear(TargetArray:Array of WBP UIKit Sort Filter Entry Object References=Get EntryPool())
   n3 = Cast To WBP_UIKit_SortFilter_Entry(execute=@n2.LoopBody, Object=@n2)
   > Unbind all Events from On Clicked!(execute:Exec, self:WBP UIKit Sort Filter Entry Object Reference=@n3) -> (then:Exec) {is=title K2Node_ClearDelegate}
  ]}
 SetColumnCount {nodes [
   n1 = SetColumnCount!() -> (then, Columns:int) {is=title K2Node_FunctionEntry}
   > Set ColumnCount(ColumnCount:int=@n1) -> (Output_Get:int=0)
   > Clear(TargetArray:"Array of Float (single-precision)s"=Get ColumnFill!(self:Grid Panel Object Reference=Get Grid_Entries()) -> (ColumnFill:"Array of Float (single-precision)s"))
   > n2 = For Loop(LastIndex=int - int(A=Get ColumnCount() -> (ColumnCount=0)))
   > SetColumnFill!(execute, self:Grid Panel Object Reference=Get Grid_Entries(), ColumnIndex:int=@n2, Coefficient:float=1) -> (then)
  ]}
 SetEntries {nodes [
   n1 = SetEntries!() -> (then, SortFilterItemVMs:Array of Fort Sort Filter Item VM Object References) {is=title K2Node_FunctionEntry}
   > n2 = HasRequiredEntryCount!(execute, self, FilterVMs:Array of Fort Sort Filter Item VM Object References=@n1) -> (then, HasRequiredEntries:bool) {is=title K2Node_CallFunction}
   > n3 = Branch(Condition=@n2)
   n4 = Get SortFilterItemVMs()
   > UpdateEntries!(execute, self, SortFilterItemVMs:Array of Fort Sort Filter Item VM Object References=@n4) -> (then)
   CreateEntries!(execute=@n3.else, self, SortFilterItemVMs:Array of Fort Sort Filter Item VM Object References=@n4) -> (then) {is=title K2Node_CallFunction}
  ]}
 SetEntryState {nodes [
   n1 = SetEntryState!() -> (then, InputVM:Fort Sort Filter Item VM Object Reference, InputEntry:WBP UIKit Sort Filter Entry Object Reference) {is=title K2Node_FunctionEntry}
   > SetToggleType(self=Get WBP_UIKit_ButtonWithToggle(self=@n1.InputEntry), Type=NewEnumerator1)
   > Set ShowCount!(
    execute
    ShowCount:bool=Get ShowFilterCounts()
    self:WBP UIKit Sort Filter Entry Object Reference=Get InputEntry() -> (InputEntry:WBP UIKit Sort Filter Entry Object Reference)
   ) -> (then, Output_Get:bool)
   > Bind Event to On Clicked(self:WBP UIKit Sort Filter Entry Object Reference=Get InputEntry(), Delegate=Create Event()) {is=title K2Node_AddDelegate}
   > Set FortSortFilterItemVM!(
    execute
    FortSortFilterItemVM:Fort Sort Filter Item VM Object Reference=@n1
    self:WBP UIKit Sort Filter Entry Object Reference=Get InputEntry()
   ) -> (then, Output_Get:Fort Sort Filter Item VM Object Reference)
   n2 = "integer > integer"(A=Get Count!(self:Fort Sort Filter Item VM Object Reference=@n1) -> (Count:int=0))
   n3 = Get InputEntry()
   > SetIsEnabled(self=@n3, bInIsEnabled=@n2)
   > SetRenderOpacity(self=@n3, InOpacity=SelectFloat(A=1, B=0.5, bPickA=@n2))
   > Return Node!(
    execute
    UsableEntry:WBP UIKit Sort Filter Entry Object Reference=Select(Option 0:WBP UIKit Sort Filter Entry Object Reference, Option 1:WBP UIKit Sort Filter Entry Object Reference=Get InputEntry(), Index=@n2) -> (ReturnValue:WBP UIKit Sort Filter Entry Object Reference)
   )
  ]}
 SetMobileLayout {nodes [
   SetMobileLayout() {is=title K2Node_FunctionEntry}
   > Branch(Condition=IsMobileGame())
   n1 = Get!() -> (Padding_Left:float=~, Padding_Top:float=~, Padding_Right:float=~, Padding_Bottom:float=~)
   > SetPadding!(
    execute
    self:Vertical Box Slot Object Reference=SlotAsVerticalBoxSlot(Widget=Get Text_Title())
    InPadding_Left:float=@n1
    InPadding_Top:float=@n1.Padding_Top
    InPadding_Right:float=@n1.Padding_Right
    InPadding_Bottom:float=Get In Padding Bottom() -> (In Padding Bottom:float)
   ) -> (then)
  ]}
 UpdateEntries {nodes [
   n1 = UpdateEntries!() -> (then, SortFilterItemVMs:Array of Fort Sort Filter Item VM Object References) {is=title K2Node_FunctionEntry}
   > n2 = For Each Loop(Array:Array of Fort Sort Filter Item VM Object References=@n1) -> (Array Element:Fort Sort Filter Item VM Object Reference)
   > Branch(Condition=Get bIsVisible(self=@n2))
   > Branch(Condition=Is Valid Index!(
     TargetArray:Array of WBP UIKit Sort Filter Entry Object References=Get EntryPool()
     IndexToTest:int=Get ButtonPoolIndex() -> (ButtonPoolIndex:int=0)
    ) -> (ReturnValue:bool))
   > n3 = SetEntryState(
    InputVM=@n2
    InputEntry="Get (a copy)"(Array:Array of WBP UIKit Sort Filter Entry Object References=Get EntryPool(), Dimension 1=Get ButtonPoolIndex() -> (ButtonPoolIndex=0)) -> (Output:WBP UIKit Sort Filter Entry Object Reference)
   ) {is=title K2Node_CallFunction}
   > Increment Int(Value=Get ButtonPoolIndex() -> (ButtonPoolIndex=0))
   > n4 = Get!(execute:Exec) -> (then:Exec, else:Exec, FirstEntry:Widget Object Reference)
   Is Valid(exec=@n4.else, InputObject=@n3)
   > Set FirstEntry(FirstEntry=@n3)
  ]}
]
variables [
 ColumnCount {type=integer, default=3}
 DesignPreviewCount {type=integer, default=18}
 EntryPool {type="array<WBP_UIKit_SortFilter_Entry_C>", default []}
 EntryWidgetClass {type=Class, default=&/WBP_UIKit_SortFilter_Entry.WBP_UIKit_SortFilter_Entry_C}
 FirstEntry {type=Widget, default=null}
 In Padding Bottom {type=number, default=24}
 OnEntryClicked {}
 SectionTitle {type=string, default=Section Title}
 ShowFilterCounts {type=boolean, default=true}
 UseEvenlySpacedColumns {type=boolean, default=true}
]
widgets {
 info {inheritedWidgetCount=0, parentClass=&UserWidget, rootWidgetClass=&VerticalBox, widgetCount=6}
 widgets [
  VerticalBox_15 {widgetClassPath=&VerticalBox, widget=VerticalBox_15}
  Text_Title {
   widgetClassPath=&CommonTextBlock
   bIsVariable=true
   parent=VerticalBox_15
   widget=Text_Title
   properties {
    text=Section Title
    font {
     bForceMonospaced=false
     bMaterialIsStencil=false
     fontMaterial=null
     fontObject=&HeadingNow
     letterSpacing=0
     monospacedWidth=1
     size=20.25
     skewAmount=0
     typefaceFontName=75 Medium
     outlineSettings {
      bApplyOutlineToDropShadows=false
      bMiteredCorners=false
      bSeparateFillAlpha=false
      outlineMaterial=null
      outlineSize=0
      outlineColor=<rgba 0 0 0 1>
     }
    }
    shadowOffset=<0 0>
   }
   slot {class=VerticalBoxSlot, properties {padding=<ltrb 0 0 0 16>}}
  }
  ScrollBox_0 {
   widgetClassPath=&ScrollBox
   parent=VerticalBox_15
   widget=ScrollBox_0
   properties {
    scrollWhenFocusChanges=AnimatedScroll
    wheelScrollMultiplier=3
    scrollbarPadding=<ltrb 2 10 10 10>
    scrollbarThickness=<12 12>
    widgetBarStyle {
     thickness=16
     draggedThumbImage {
      bIsDynamicallyLoaded=false
      drawAs=RoundedBox
      imageType=FullColor
      mirroring=NoMirror
      resourceName=../../../Engine/Content/Slate/Common/Scrollbar_Thumb.png
      resourceObject=null
      tiling=NoTile
      imageSize=<0 0>
      margin=<ltrb 0.25 0.25 0.25 0.25>
      outlineSettings {
       bUseBrushTransparency=false
       roundingType=FixedRadius
       width=2
       color {colorUseRule=UseColor_Specified, specifiedColor=<rgba 0 0 0 0>}
       cornerRadii=<6 6 6 6>
      }
      tintColor {colorUseRule=UseColor_Specified, specifiedColor=<rgba 1 1 1 1>}
      uVRegion {bIsValid=false, max=<0 0>, min=<0 0>}
     }
     horizontalBackgroundImage {
      bIsDynamicallyLoaded=false
      drawAs=NoDrawType
      imageType=NoImage
      mirroring=NoMirror
      resourceName=None
      resourceObject=null
      tiling=NoTile
      imageSize=<0 0>
      margin=<ltrb 0 0 0 0>
      outlineSettings {
       bUseBrushTransparency=false
       roundingType=HalfHeightRadius
       width=0
       color {colorUseRule=UseColor_Specified, specifiedColor=<rgba 0 0 0 0>}
       cornerRadii=<0 0 0 0>
      }
      tintColor {colorUseRule=UseColor_Specified, specifiedColor=<rgba 1 1 1 1>}
      uVRegion {bIsValid=false, max=<0 0>, min=<0 0>}
     }
     horizontalBottomSlotImage {
      bIsDynamicallyLoaded=false
      drawAs=NoDrawType
      imageType=FullColor
      mirroring=NoMirror
      resourceName=../../../Engine/Content/Slate/Common/Scrollbar_Background_Horizontal.png
      resourceObject=null
      tiling=NoTile
      imageSize=<8 8>
      margin=<ltrb 0 0 0 0>
      outlineSettings {
       bUseBrushTransparency=false
       roundingType=HalfHeightRadius
       width=0
       color {colorUseRule=UseColor_Specified, specifiedColor=<rgba 0 0 0 0>}
       cornerRadii=<0 0 0 0>
      }
      tintColor {colorUseRule=UseColor_Specified, specifiedColor=<rgba 1 1 1 1>}
      uVRegion {bIsValid=false, max=<0 0>, min=<0 0>}
     }
     horizontalTopSlotImage {
      bIsDynamicallyLoaded=false
      drawAs=NoDrawType
      imageType=FullColor
      mirroring=NoMirror
      resourceName=../../../Engine/Content/Slate/Common/Scrollbar_Background_Horizontal.png
      resourceObject=null
      tiling=NoTile
      imageSize=<8 8>
      margin=<ltrb 0 0 0 0>
      outlineSettings {
       bUseBrushTransparency=false
       roundingType=HalfHeightRadius
       width=0
       color {colorUseRule=UseColor_Specified, specifiedColor=<rgba 0 0 0 0>}
       cornerRadii=<0 0 0 0>
      }
      tintColor {colorUseRule=UseColor_Specified, specifiedColor=<rgba 1 1 1 1>}
      uVRegion {bIsValid=false, max=<0 0>, min=<0 0>}
     }
     hoveredThumbImage {
      bIsDynamicallyLoaded=false
      drawAs=RoundedBox
      imageType=FullColor
      mirroring=NoMirror
      resourceName=../../../Engine/Content/Slate/Common/Scrollbar_Thumb.png
      resourceObject=null
      tiling=NoTile
      imageSize=<0 0>
      margin=<ltrb 0.25 0.25 0.25 0.25>
      outlineSettings {
       bUseBrushTransparency=false
       roundingType=FixedRadius
       width=2
       color {colorUseRule=UseColor_Specified, specifiedColor=<rgba 0 0 0 0>}
       cornerRadii=<6 6 6 6>
      }
      tintColor {colorUseRule=UseColor_Specified, specifiedColor=<rgba 1 1 1 1>}
      uVRegion {bIsValid=false, max=<0 0>, min=<0 0>}
     }
     normalThumbImage {
      bIsDynamicallyLoaded=false
      drawAs=RoundedBox
      imageType=FullColor
      mirroring=NoMirror
      resourceName=../../../Engine/Content/Slate/Common/Scrollbar_Thumb.png
      resourceObject=null
      tiling=NoTile
      imageSize=<0 0>
      margin=<ltrb 0.25 0.25 0.25 0.25>
      outlineSettings {
       bUseBrushTransparency=false
       roundingType=FixedRadius
       width=2
       color {colorUseRule=UseColor_Specified, specifiedColor=<rgba 0 0 0 0>}
       cornerRadii=<6 6 6 6>
      }
      tintColor {colorUseRule=UseColor_Specified, specifiedColor=<rgba 1 1 1 0.749>}
      uVRegion {bIsValid=false, max=<0 0>, min=<0 0>}
     }
     verticalBackgroundImage {
      bIsDynamicallyLoaded=false
      drawAs=Box
      imageType=FullColor
      mirroring=NoMirror
      resourceName=None
      resourceObject=/FortUILibrary/UIKit/ScrollBar/MaterialInstances/MI_UIKit_ScrollbarBackground
      tiling=NoTile
      imageSize=<12 12>
      margin=<ltrb 0.5 0.5 0.5 0.5>
      outlineSettings {
       bUseBrushTransparency=false
       roundingType=HalfHeightRadius
       width=0
       color {colorUseRule=UseColor_Specified, specifiedColor=<rgba 0 0 0 0>}
       cornerRadii=<0 0 0 0>
      }
      tintColor {colorUseRule=UseColor_Specified, specifiedColor=<rgba 0.521 0.521 0.521 0.502>}
      uVRegion {bIsValid=false, max=<0 0>, min=<0 0>}
     }
     verticalBottomSlotImage {
      bIsDynamicallyLoaded=false
      drawAs=NoDrawType
      imageType=FullColor
      mirroring=NoMirror
      resourceName=../../../Engine/Content/Slate/Common/Scrollbar_Background_Vertical.png
      resourceObject=null
      tiling=NoTile
      imageSize=<8 8>
      margin=<ltrb 0 0 0 0>
      outlineSettings {
       bUseBrushTransparency=false
       roundingType=HalfHeightRadius
       width=0
       color {colorUseRule=UseColor_Specified, specifiedColor=<rgba 0 0 0 0>}
       cornerRadii=<0 0 0 0>
      }
      tintColor {colorUseRule=UseColor_Specified, specifiedColor=<rgba 1 1 1 1>}
      uVRegion {bIsValid=false, max=<0 0>, min=<0 0>}
     }
     verticalTopSlotImage {
      bIsDynamicallyLoaded=false
      drawAs=NoDrawType
      imageType=FullColor
      mirroring=NoMirror
      resourceName=../../../Engine/Content/Slate/Common/Scrollbar_Background_Vertical.png
      resourceObject=null
      tiling=NoTile
      imageSize=<8 8>
      margin=<ltrb 0 0 0 0>
      outlineSettings {
       bUseBrushTransparency=false
       roundingType=HalfHeightRadius
       width=0
       color {colorUseRule=UseColor_Specified, specifiedColor=<rgba 0 0 0 0>}
       cornerRadii=<0 0 0 0>
      }
      tintColor {colorUseRule=UseColor_Specified, specifiedColor=<rgba 1 1 1 1>}
      uVRegion {bIsValid=false, max=<0 0>, min=<0 0>}
     }
    }
    widgetStyle {
     barThickness=12
     bottomShadowBrush {
      bIsDynamicallyLoaded=false
      drawAs=NoDrawType
      imageType=FullColor
      mirroring=NoMirror
      resourceName=../../../Engine/Content/Slate/Common/ScrollBoxShadowBottom.png
      resourceObject=null
      tiling=NoTile
      imageSize=<16 8>
      margin=<ltrb 0.5 0 0.5 1>
      outlineSettings {
       bUseBrushTransparency=false
       roundingType=HalfHeightRadius
       width=0
       color {colorUseRule=UseColor_Specified, specifiedColor=<rgba 0 0 0 0>}
       cornerRadii=<0 0 0 0>
      }
      tintColor {colorUseRule=UseColor_Specified, specifiedColor=<rgba 1 1 1 1>}
      uVRegion {bIsValid=false, max=<0 0>, min=<0 0>}
     }
     horizontalScrolledContentPadding=<ltrb 0 0 1 0>
     leftShadowBrush {
      bIsDynamicallyLoaded=false
      drawAs=NoDrawType
      imageType=FullColor
      mirroring=NoMirror
      resourceName=../../../Engine/Content/Slate/Common/ScrollBoxShadowLeft.png
      resourceObject=null
      tiling=NoTile
      imageSize=<8 16>
      margin=<ltrb 1 0.5 0 0.5>
      outlineSettings {
       bUseBrushTransparency=false
       roundingType=HalfHeightRadius
       width=0
       color {colorUseRule=UseColor_Specified, specifiedColor=<rgba 0 0 0 0>}
       cornerRadii=<0 0 0 0>
      }
      tintColor {colorUseRule=UseColor_Specified, specifiedColor=<rgba 1 1 1 1>}
      uVRegion {bIsValid=false, max=<0 0>, min=<0 0>}
     }
     rightShadowBrush {
      bIsDynamicallyLoaded=false
      drawAs=NoDrawType
      imageType=FullColor
      mirroring=NoMirror
      resourceName=../../../Engine/Content/Slate/Common/ScrollBoxShadowRight.png
      resourceObject=null
      tiling=NoTile
      imageSize=<8 16>
      margin=<ltrb 0 0.5 1 0.5>
      outlineSettings {
       bUseBrushTransparency=false
       roundingType=HalfHeightRadius
       width=0
       color {colorUseRule=UseColor_Specified, specifiedColor=<rgba 0 0 0 0>}
       cornerRadii=<0 0 0 0>
      }
      tintColor {colorUseRule=UseColor_Specified, specifiedColor=<rgba 1 1 1 1>}
      uVRegion {bIsValid=false, max=<0 0>, min=<0 0>}
     }
     topShadowBrush {
      bIsDynamicallyLoaded=false
      drawAs=NoDrawType
      imageType=FullColor
      mirroring=NoMirror
      resourceName=../../../Engine/Content/Slate/Common/ScrollBoxShadowTop.png
      resourceObject=null
      tiling=NoTile
      imageSize=<16 8>
      margin=<ltrb 0.5 1 0.5 0>
      outlineSettings {
       bUseBrushTransparency=false
       roundingType=HalfHeightRadius
       width=0
       color {colorUseRule=UseColor_Specified, specifiedColor=<rgba 0 0 0 0>}
       cornerRadii=<0 0 0 0>
      }
      tintColor {colorUseRule=UseColor_Specified, specifiedColor=<rgba 1 1 1 1>}
      uVRegion {bIsValid=false, max=<0 0>, min=<0 0>}
     }
     verticalScrolledContentPadding=<ltrb 0 0 0 1>
    }
   }
   slot {class=VerticalBoxSlot, properties {padding=<ltrb -8 -8 -8 -8>, size {sizeRule=Fill, value=1}}}
  }
  Switcher_Entries {
   widgetClassPath=&CommonVisibilitySwitcher
   bIsVariable=true
   parent=ScrollBox_0
   widget=Switcher_Entries
   slot {class=ScrollBoxSlot, properties {padding=<ltrb 8 8 8 8>}}
  }
  Grid_Entries {
   widgetClassPath=/Script/FortniteUI.UIKitGridPanel
   bIsVariable=true
   parent=Switcher_Entries
   widget=Grid_Entries
   slot {class=CommonVisibilitySwitcherSlot}
  }
  DynamicBox_Entries {
   widgetClassPath=&DynamicEntryBox
   bIsVariable=true
   parent=Switcher_Entries
   widget=DynamicBox_Entries
   properties {
    entryBoxType=Wrap
    entryWidgetClass=&WBP_UIKit_SortFilter_Entry_C
    numDesignerPreviewEntries=9
    entrySpacing=<16 0>
   }
   slot {class=CommonVisibilitySwitcherSlot}
  }
 ]
}
